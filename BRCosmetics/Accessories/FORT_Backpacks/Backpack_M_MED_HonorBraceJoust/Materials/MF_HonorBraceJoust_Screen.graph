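# Material function for the HonorBraceJoust backpack screen, layering an animated flipbook face onto incoming material attributes.
# Three flipbook samples composite into the face, with time-cycled Normal and Smile frames, Reactivity_Toggle switching to Yell, and sine/cosine UV drift at differing scales for depth.
# The green TechArtMask channel and Fresnel input feed screen masking and falloff, adjusting base color, specular, roughness and emissive.
# A sphere mask adds a pulsing red emissive dot.
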MaterialFunction /BRCosmetics/Accessories/FORT_Backpacks/Backpack_M_MED_HonorBraceJoust/Materials/MF_HonorBraceJoust_Screen
graph {expressions {
  MaterialExpressionAdd_0 {
   class=MaterialExpressionAdd
   inputs [
    A {from=MaterialExpressionMaterialFunctionCall_12, output=Rotated Values}
    B {from=MaterialExpressionConvert_0}
   ]
   properties {constB=0}
  }
  MaterialExpressionAdd_1 {
   class=MaterialExpressionAdd
   inputs [
    A {from=MaterialExpressionModulo_2}
    B {from=""}
   ]
   properties {constB=2.222}
  }
  MaterialExpressionAdd_2 {
   class=MaterialExpressionAdd
   inputs [
    A {from=MaterialExpressionModulo_3}
    B {from=""}
   ]
   properties {constB=4.444}
  }
  MaterialExpressionAdd_4 {
   class=MaterialExpressionAdd
   inputs [
    A {from=MaterialExpressionMaterialFunctionCall_12, output=Rotated Values}
    B {from=MaterialExpressionMultiply_2}
   ]
   properties {constB=0}
  }
  MaterialExpressionAdd_5 {
   class=MaterialExpressionAdd
   inputs [
    A {from=MaterialExpressionMaterialFunctionCall_12, output=Rotated Values}
    B {from=MaterialExpressionMultiply_3}
   ]
   properties {constB=0}
  }
  MaterialExpressionAdd_8 {
   class=MaterialExpressionAdd
   inputs [
    A {from=MaterialExpressionNamedRerouteUsage_7}
    B {from=MaterialExpressionNamedRerouteUsage_8}
   ]
  }
  MaterialExpressionAdd_9 {
   class=MaterialExpressionAdd
   inputs [
    A {from=MaterialExpressionMultiply_12}
    B {from=MaterialExpressionAdd_8}
   ]
  }
  MaterialExpressionComponentMask_0 {class=MaterialExpressionComponentMask, inputs [None {from=MaterialExpressionFunctionInput_1}], properties {g=true}}
  MaterialExpressionConstant2Vector_0 {class=MaterialExpressionConstant2Vector, properties {desc=FlipbookSize, g=3, r=3}}
  MaterialExpressionConstant2Vector_9 {class=MaterialExpressionConstant2Vector, properties {g=0.988, r=0.905}}
  MaterialExpressionConstant3Vector_1 {class=MaterialExpressionConstant3Vector, properties {constant=<rgba 4 0 0 0>}}
  MaterialExpressionConstantBiasScale_0 {class=MaterialExpressionConstantBiasScale, inputs [None {from=MaterialExpressionSine_0}]}
  MaterialExpressionConstantBiasScale_1 {class=MaterialExpressionConstantBiasScale, inputs [None {from=MaterialExpressionCosine_0}]}
  MaterialExpressionConstant_0 {class=MaterialExpressionConstant, properties {r=0.9}}
  MaterialExpressionConstant_12 {class=MaterialExpressionConstant, properties {r=0.5}}
  MaterialExpressionConstant_16 {class=MaterialExpressionConstant}
  MaterialExpressionConstant_2 {class=MaterialExpressionConstant, properties {r=2}}
  MaterialExpressionConstant_5 {class=MaterialExpressionConstant, properties {r=0.8}}
  MaterialExpressionConstant_6 {class=MaterialExpressionConstant, properties {r=0.7}}
  MaterialExpressionConstant_7 {class=MaterialExpressionConstant, properties {r=4}}
  MaterialExpressionConvert_0 {
   class=MaterialExpressionConvert
   inputs [
    None {from=MaterialExpressionDivide_2}
    None {from=MaterialExpressionDivide_0}
   ]
   properties {
    convertInputs [
     {
      type=Scalar
      expressionInput {
       expression=&self:MaterialExpressionDivide_2
       inputName=None
       mask=0
       maskA=0
       maskB=0
       maskG=0
       maskR=0
       outputIndex=0
      }
     }
     {
      type=Scalar
      expressionInput {
       expression=&self:MaterialExpressionDivide_0
       inputName=None
       mask=0
       maskA=0
       maskB=0
       maskG=0
       maskR=0
       outputIndex=0
      }
     }
    ]
    convertOutputs [{type=Vector2}]
   }
  }
  MaterialExpressionCosine_0 {class=MaterialExpressionCosine, inputs [None {from=MaterialExpressionTime_0, output=""}], properties {}}
  MaterialExpressionDivide_0 {
   class=MaterialExpressionDivide
   inputs [
    A {from=MaterialExpressionLinearInterpolate_1}
    B {from=""}
   ]
   properties {constB=100}
  }
  MaterialExpressionDivide_1 {
   class=MaterialExpressionDivide
   inputs [
    A {from=MaterialExpressionModulo_0}
    B {from=""}
   ]
   properties {constB=10}
  }
  MaterialExpressionDivide_2 {
   class=MaterialExpressionDivide
   inputs [
    A {from=MaterialExpressionLinearInterpolate_2}
    B {from=""}
   ]
   properties {constB=100}
  }
  MaterialExpressionDivide_3 {
   class=MaterialExpressionDivide
   inputs [
    A {from=MaterialExpressionAdd_1}
    B {from=""}
   ]
   properties {constB=10}
  }
  MaterialExpressionDivide_4 {
   class=MaterialExpressionDivide
   inputs [
    A {from=MaterialExpressionAdd_2}
    B {from=""}
   ]
   properties {constB=10}
  }
  MaterialExpressionFunctionInput_0 {
   class=MaterialExpressionFunctionInput
   inputs [Preview {from="", output=""}]
   properties {inputName=OUT_MA, inputType=FunctionInput_MaterialAttributes}
  }
  MaterialExpressionFunctionInput_1 {
   class=MaterialExpressionFunctionInput
   inputs [Preview {from="", output=""}]
   properties {bUsePreviewValueAsDefault=true, inputName=TechArtMask, sortPriority=1, previewValue=<1 1 1 1>}
  }
  MaterialExpressionFunctionInput_2 {
   class=MaterialExpressionFunctionInput
   inputs [Preview {from="", output=""}]
   properties {bUsePreviewValueAsDefault=true, inputName=Fresnel, inputType=FunctionInput_Scalar, sortPriority=3, previewValue=<1 1 1 1>}
  }
  MaterialExpressionFunctionOutput_1 {
   class=MaterialExpressionFunctionOutput
   inputs [None {from=MaterialExpressionSetMaterialAttributes_0, output=""}]
   properties {outputName=OUT_MA}
  }
  MaterialExpressionGetMaterialAttributes_0 {
   class=MaterialExpressionGetMaterialAttributes
   inputs [None {from=MaterialExpressionFunctionInput_0, output=""}]
   properties {attributeGetTypes [
     #1
     #2
     #3
     #4
     #5
     #6
    ]}
  }
  MaterialExpressionLinearInterpolate_1 {
   class=MaterialExpressionLinearInterpolate
   inputs [
    A {from=""}
    B {from=""}
    Alpha {from=MaterialExpressionConstantBiasScale_0}
   ]
   properties {constA=-0.1, constB=-0.5}
  }
  MaterialExpressionLinearInterpolate_10 {
   class=MaterialExpressionLinearInterpolate
   inputs [
    A {from=MaterialExpressionNamedRerouteUsage_12}
    B {from=""}
    Alpha {from=MaterialExpressionNamedRerouteUsage_11}
   ]
   properties {constB=0}
  }
  MaterialExpressionLinearInterpolate_11 {
   class=MaterialExpressionLinearInterpolate
   inputs [
    A {from=MaterialExpressionNamedRerouteUsage_14}
    B {from=MaterialExpressionScalarParameter_3}
    Alpha {from=MaterialExpressionNamedRerouteUsage_13}
   ]
   properties {constB=0.2}
  }
  MaterialExpressionLinearInterpolate_2 {
   class=MaterialExpressionLinearInterpolate
   inputs [
    A {from=""}
    B {from=""}
    Alpha {from=MaterialExpressionConstantBiasScale_1}
   ]
   properties {constA=-0.1, constB=-0.5}
  }
  MaterialExpressionLinearInterpolate_4 {
   class=MaterialExpressionLinearInterpolate
   inputs [
    A {from=MaterialExpressionTextureSample_1, output=RGB}
    B {from=MaterialExpressionTextureSample_2, output=RGB}
    Alpha {from=MaterialExpressionTextureSample_2, output=RGB}
   ]
  }
  MaterialExpressionLinearInterpolate_5 {
   class=MaterialExpressionLinearInterpolate
   inputs [
    A {from=MaterialExpressionLinearInterpolate_4}
    B {from=MaterialExpressionTextureSample_5, output=RGB}
    Alpha {from=MaterialExpressionTextureSample_5, output=RGB}
   ]
  }
  MaterialExpressionLinearInterpolate_9 {
   class=MaterialExpressionLinearInterpolate
   inputs [
    A {from=MaterialExpressionNamedRerouteUsage_10}
    B {from=""}
    Alpha {from=MaterialExpressionNamedRerouteUsage_9}
   ]
   properties {constB=0}
  }
  MaterialExpressionMaterialFunctionCall_0 {
   class=MaterialExpressionMaterialFunctionCall
   inputs [
    Size {from=MaterialExpressionConstant2Vector_0}
    Selector {from=MaterialExpressionSwitch_3}
    UV {from=MaterialExpressionSwitch_6}
   ]
   properties {materialFunction=/Game/Global/MaterialFunctions/MF_FlipbookUV}
  }
  MaterialExpressionMaterialFunctionCall_1 {
   class=MaterialExpressionMaterialFunctionCall
   inputs [
    Size {from=MaterialExpressionConstant2Vector_0}
    Selector {from=MaterialExpressionSwitch_4}
    UV {from=MaterialExpressionMaterialFunctionCall_12, output=Rotated Values}
   ]
   properties {materialFunction=&MF_FlipbookUV}
  }
  MaterialExpressionMaterialFunctionCall_12 {
   class=MaterialExpressionMaterialFunctionCall
   inputs [
    UVs {from=MaterialExpressionTextureCoordinate_1}
    Rotation Center {from=""}
    "Rotation Angle (0-1)" {from=MaterialExpressionConstant_12}
   ]
   properties {materialFunction=&CustomRotator}
  }
  MaterialExpressionMaterialFunctionCall_7 {
   class=MaterialExpressionMaterialFunctionCall
   inputs [
    Size {from=MaterialExpressionConstant2Vector_0}
    Selector {from=MaterialExpressionConstant_6}
    UV {from=MaterialExpressionSwitch_7}
   ]
   properties {materialFunction=&MF_FlipbookUV}
  }
  MaterialExpressionModulo_0 {
   class=MaterialExpressionModulo
   inputs [
    A {from=MaterialExpressionMultiply_5, output=""}
    B {from=MaterialExpressionConstant_2, output=""}
   ]
   properties {}
  }
  MaterialExpressionModulo_2 {
   class=MaterialExpressionModulo
   inputs [
    A {from=MaterialExpressionMultiply_5, output=""}
    B {from=MaterialExpressionConstant_2, output=""}
   ]
   properties {}
  }
  MaterialExpressionModulo_3 {
   class=MaterialExpressionModulo
   inputs [
    A {from=MaterialExpressionMultiply_4, output=""}
    B {from=MaterialExpressionConstant_2, output=""}
   ]
   properties {}
  }
  MaterialExpressionModulo_4 {
   class=MaterialExpressionModulo
   inputs [
    A {from=MaterialExpressionTime_4, output=""}
    B {from=MaterialExpressionConstant_7, output=""}
   ]
   properties {}
  }
  MaterialExpressionMultiply_0 {
   class=MaterialExpressionMultiply
   inputs [
    A {from=MaterialExpressionSine_1}
    B {from=""}
   ]
   properties {constB=2}
  }
  MaterialExpressionMultiply_11 {
   class=MaterialExpressionMultiply
   inputs [
    A {from=MaterialExpressionSphereMask_0}
    B {from=MaterialExpressionSubtract_0}
   ]
  }
  MaterialExpressionMultiply_12 {
   class=MaterialExpressionMultiply
   inputs [
    A {from=MaterialExpressionMultiply_11}
    B {from=MaterialExpressionConstant3Vector_1}
   ]
   properties {constB=10}
  }
  MaterialExpressionMultiply_2 {
   class=MaterialExpressionMultiply
   inputs [
    A {from=MaterialExpressionConvert_0}
    B {from=""}
   ]
   properties {constB=4}
  }
  MaterialExpressionMultiply_3 {
   class=MaterialExpressionMultiply
   inputs [
    A {from=MaterialExpressionConvert_0}
    B {from=""}
   ]
   properties {constB=8}
  }
  MaterialExpressionMultiply_4 {
   class=MaterialExpressionMultiply
   inputs [
    A {from=MaterialExpressionTime_3}
    B {from=""}
   ]
   properties {constB=0.25}
  }
  MaterialExpressionMultiply_5 {
   class=MaterialExpressionMultiply
   inputs [
    A {from=MaterialExpressionTime_1}
    B {from=""}
   ]
   properties {constB=2}
  }
  MaterialExpressionMultiply_8 {
   class=MaterialExpressionMultiply
   inputs [
    A {from=MaterialExpressionMultiply_9}
    B {from=MaterialExpressionScalarParameter_1}
   ]
  }
  MaterialExpressionMultiply_9 {
   class=MaterialExpressionMultiply
   inputs [
    A {from=MaterialExpressionLinearInterpolate_5}
    B {from=MaterialExpressionPower_0}
   ]
  }
  MaterialExpressionNamedRerouteDeclaration_0 {
   class=MaterialExpressionNamedRerouteDeclaration
   inputs [None {from=MaterialExpressionLinearInterpolate_9}]
   properties {name=OUT_BaseColor, nodeColor=<rgba 0 0.8856 1 1>}
  }
  MaterialExpressionNamedRerouteDeclaration_1 {class=MaterialExpressionNamedRerouteDeclaration, inputs [None {from=""}], properties {name=OUT_Metallic, nodeColor=<rgba 0.3462 1 0.0006 1>}}
  MaterialExpressionNamedRerouteDeclaration_10 {
   class=MaterialExpressionNamedRerouteDeclaration
   inputs [None {from=MaterialExpressionGetMaterialAttributes_0, output=Normal}]
   properties {name=IN_Normal, nodeColor=<rgba 0.25 0.25 1 1>}
  }
  MaterialExpressionNamedRerouteDeclaration_11 {
   class=MaterialExpressionNamedRerouteDeclaration
   inputs [None {from=MaterialExpressionGetMaterialAttributes_0, output=EmissiveColor}]
   properties {name=IN_EmissiveColor, nodeColor=<rgba 0 0.8 1 1>}
  }
  MaterialExpressionNamedRerouteDeclaration_12 {
   class=MaterialExpressionNamedRerouteDeclaration
   inputs [None {from=MaterialExpressionMultiply_8}]
   properties {name=ScreenEmissive, nodeColor=<rgba 0 1 0.5781 1>}
  }
  MaterialExpressionNamedRerouteDeclaration_13 {
   class=MaterialExpressionNamedRerouteDeclaration
   inputs [None {from=MaterialExpressionComponentMask_0}]
   properties {name=ScreenMask, nodeColor=<rgba 1 0 0.5625 1>}
  }
  MaterialExpressionNamedRerouteDeclaration_14 {
   class=MaterialExpressionNamedRerouteDeclaration
   inputs [None {from=MaterialExpressionFunctionInput_2}]
   properties {name=IN_Fresnel, nodeColor=<rgba 1 0.8203 0 1>}
  }
  MaterialExpressionNamedRerouteDeclaration_2 {
   class=MaterialExpressionNamedRerouteDeclaration
   inputs [None {from=MaterialExpressionLinearInterpolate_10}]
   properties {name=OUT_Specular, nodeColor=<rgba 0.25 1 0 1>}
  }
  MaterialExpressionNamedRerouteDeclaration_3 {
   class=MaterialExpressionNamedRerouteDeclaration
   inputs [None {from=MaterialExpressionLinearInterpolate_11}]
   properties {name=OUT_Roughness, nodeColor=<rgba 0.25 1 0 1>}
  }
  MaterialExpressionNamedRerouteDeclaration_4 {class=MaterialExpressionNamedRerouteDeclaration, inputs [None {from=""}], properties {name=OUT_Normal, nodeColor=<rgba 0.25 0.25 1 1>}}
  MaterialExpressionNamedRerouteDeclaration_5 {
   class=MaterialExpressionNamedRerouteDeclaration
   inputs [None {from=MaterialExpressionAdd_9}]
   properties {name=OUT_EmissiveColor, nodeColor=<rgba 0 0.8 1 1>}
  }
  MaterialExpressionNamedRerouteDeclaration_6 {
   class=MaterialExpressionNamedRerouteDeclaration
   inputs [None {from=MaterialExpressionGetMaterialAttributes_0, output=BaseColor}]
   properties {name=IN_BaseColor, nodeColor=<rgba 0 0.8 1 1>}
  }
  MaterialExpressionNamedRerouteDeclaration_7 {
   class=MaterialExpressionNamedRerouteDeclaration
   inputs [None {from=MaterialExpressionGetMaterialAttributes_0, output=Metallic}]
   properties {name=IN_Metallic, nodeColor=<rgba 0.25 1 0 1>}
  }
  MaterialExpressionNamedRerouteDeclaration_8 {
   class=MaterialExpressionNamedRerouteDeclaration
   inputs [None {from=MaterialExpressionGetMaterialAttributes_0, output=Specular}]
   properties {name=IN_Specular, nodeColor=<rgba 0.25 1 0 1>}
  }
  MaterialExpressionNamedRerouteDeclaration_9 {
   class=MaterialExpressionNamedRerouteDeclaration
   inputs [None {from=MaterialExpressionGetMaterialAttributes_0, output=Roughness}]
   properties {name=IN_Roughness, nodeColor=<rgba 0.25 1 0 1>}
  }
  MaterialExpressionNamedRerouteUsage_0 {class=MaterialExpressionNamedRerouteUsage}
  MaterialExpressionNamedRerouteUsage_1 {class=MaterialExpressionNamedRerouteUsage}
  MaterialExpressionNamedRerouteUsage_10 {class=MaterialExpressionNamedRerouteUsage}
  MaterialExpressionNamedRerouteUsage_11 {class=MaterialExpressionNamedRerouteUsage}
  MaterialExpressionNamedRerouteUsage_12 {class=MaterialExpressionNamedRerouteUsage}
  MaterialExpressionNamedRerouteUsage_13 {class=MaterialExpressionNamedRerouteUsage}
  MaterialExpressionNamedRerouteUsage_14 {class=MaterialExpressionNamedRerouteUsage}
  MaterialExpressionNamedRerouteUsage_15 {class=MaterialExpressionNamedRerouteUsage}
  MaterialExpressionNamedRerouteUsage_2 {class=MaterialExpressionNamedRerouteUsage}
  MaterialExpressionNamedRerouteUsage_3 {class=MaterialExpressionNamedRerouteUsage}
  MaterialExpressionNamedRerouteUsage_4 {class=MaterialExpressionNamedRerouteUsage}
  MaterialExpressionNamedRerouteUsage_5 {class=MaterialExpressionNamedRerouteUsage}
  MaterialExpressionNamedRerouteUsage_7 {class=MaterialExpressionNamedRerouteUsage}
  MaterialExpressionNamedRerouteUsage_8 {class=MaterialExpressionNamedRerouteUsage}
  MaterialExpressionNamedRerouteUsage_9 {class=MaterialExpressionNamedRerouteUsage}
  MaterialExpressionOneMinus_0 {class=MaterialExpressionOneMinus, inputs [None {from=MaterialExpressionNamedRerouteUsage_15}]}
  MaterialExpressionPower_0 {
   class=MaterialExpressionPower
   inputs [
    Base {from=MaterialExpressionOneMinus_0}
    Exp {from=""}
   ]
   properties {constExponent=1.8}
  }
  MaterialExpressionScalarParameter_0 {class=MaterialExpressionScalarParameter, properties {group=HornorBraceJoust, parameterName=Reactivity_Toggle, sliderMax=1}}
  MaterialExpressionScalarParameter_1 {class=MaterialExpressionScalarParameter, properties {defaultValue=1.5, group=HornorBraceJoust, parameterName=Screen_Brightness}}
  MaterialExpressionScalarParameter_3 {class=MaterialExpressionScalarParameter, properties {defaultValue=0.2, group=HornorBraceJoust, parameterName=Screen_Roughness}}
  MaterialExpressionSetMaterialAttributes_0 {
   class=MaterialExpressionSetMaterialAttributes
   inputs [
    MaterialAttributes {from=MaterialExpressionGetMaterialAttributes_0, output=MaterialAttributes}
    BaseColor {from=MaterialExpressionNamedRerouteUsage_0, output=""}
    Metallic {from=MaterialExpressionGetMaterialAttributes_0, output=MaterialAttributes}
    Specular {from=MaterialExpressionNamedRerouteUsage_2, output=""}
    Roughness {from=MaterialExpressionNamedRerouteUsage_3, output=""}
    Normal {from=MaterialExpressionGetMaterialAttributes_0, output=MaterialAttributes}
    EmissiveColor {from=MaterialExpressionNamedRerouteUsage_4, output=""}
   ]
   properties {attributeSetTypes [
     #1
     #2
     #3
     #4
     #5
     #6
    ]}
  }
  MaterialExpressionSine_0 {class=MaterialExpressionSine, inputs [None {from=MaterialExpressionTime_0}]}
  MaterialExpressionSine_1 {class=MaterialExpressionSine, inputs [None {from=MaterialExpressionTime_2}]}
  MaterialExpressionSphereMask_0 {
   class=MaterialExpressionSphereMask
   inputs [
    A {from=MaterialExpressionTextureCoordinate_3, output=""}
    B {from=MaterialExpressionConstant2Vector_9, output=""}
    Radius {from="", output=""}
    Hardness {from="", output=""}
   ]
   properties {attenuationRadius=0.015, hardnessPercent=10}
  }
  MaterialExpressionSubtract_0 {
   class=MaterialExpressionSubtract
   inputs [
    A {from=MaterialExpressionMultiply_0}
    B {from=""}
   ]
  }
  MaterialExpressionSwitch_0 {
   class=MaterialExpressionSwitch
   inputs [
    SwitchValue {from=MaterialExpressionModulo_4, output=""}
    Default {from=MaterialExpressionDivide_1, output=""}
    Normal {from=MaterialExpressionDivide_1, output=""}
    Smile {from=MaterialExpressionDivide_3, output=""}
    Normal {from=MaterialExpressionDivide_1, output=""}
   ]
   properties {inputs [
     {
      inputName=Normal
      input {
       expression=&self:MaterialExpressionDivide_1
       inputName=None
       mask=0
       maskA=0
       maskB=0
       maskG=0
       maskR=0
       outputIndex=0
      }
     }
     {
      inputName=Smile
      input {
       expression=&self:MaterialExpressionDivide_3
       inputName=None
       mask=0
       maskA=0
       maskB=0
       maskG=0
       maskR=0
       outputIndex=0
      }
     }
     {
      inputName=Normal
      input {
       expression=&self:MaterialExpressionDivide_1
       inputName=None
       mask=0
       maskA=0
       maskB=0
       maskG=0
       maskR=0
       outputIndex=0
      }
     }
    ]}
  }
  MaterialExpressionSwitch_3 {
   class=MaterialExpressionSwitch
   inputs [
    SwitchValue {from=MaterialExpressionScalarParameter_0, output=""}
    Default {from=MaterialExpressionSwitch_0, output=""}
    Normal {from=MaterialExpressionSwitch_0, output=""}
    Yell {from=MaterialExpressionDivide_4, output=""}
   ]
   properties {inputs [
     {
      inputName=Normal
      input {
       expression=&self:MaterialExpressionSwitch_0
       inputName=None
       mask=0
       maskA=0
       maskB=0
       maskG=0
       maskR=0
       outputIndex=0
      }
     }
     {
      inputName=Yell
      input {
       expression=&self:MaterialExpressionDivide_4
       inputName=None
       mask=0
       maskA=0
       maskB=0
       maskG=0
       maskR=0
       outputIndex=0
      }
     }
    ]}
  }
  MaterialExpressionSwitch_4 {
   class=MaterialExpressionSwitch
   inputs [
    SwitchValue {from=MaterialExpressionScalarParameter_0, output=""}
    Default {from=MaterialExpressionConstant_5, output=""}
    Normal {from=MaterialExpressionConstant_5, output=""}
    Yell {from=MaterialExpressionConstant_0, output=""}
   ]
   properties {inputs [
     {
      inputName=Normal
      input {
       expression=&self:MaterialExpressionConstant_5
       inputName=None
       mask=0
       maskA=0
       maskB=0
       maskG=0
       maskR=0
       outputIndex=0
      }
     }
     {
      inputName=Yell
      input {
       expression=&self:MaterialExpressionConstant_0
       inputName=None
       mask=0
       maskA=0
       maskB=0
       maskG=0
       maskR=0
       outputIndex=0
      }
     }
    ]}
  }
  MaterialExpressionSwitch_6 {
   class=MaterialExpressionSwitch
   inputs [
    SwitchValue {from=MaterialExpressionScalarParameter_0, output=""}
    Default {from=MaterialExpressionAdd_0, output=""}
    Normal {from=MaterialExpressionAdd_4, output=""}
    Yell {from=MaterialExpressionAdd_5, output=""}
   ]
   properties {inputs [
     {
      inputName=Normal
      input {
       expression=&self:MaterialExpressionAdd_4
       inputName=None
       mask=0
       maskA=0
       maskB=0
       maskG=0
       maskR=0
       outputIndex=0
      }
     }
     {
      inputName=Yell
      input {
       expression=&self:MaterialExpressionAdd_5
       inputName=None
       mask=0
       maskA=0
       maskB=0
       maskG=0
       maskR=0
       outputIndex=0
      }
     }
    ]}
  }
  MaterialExpressionSwitch_7 {
   class=MaterialExpressionSwitch
   inputs [
    SwitchValue {from=MaterialExpressionScalarParameter_0, output=""}
    Default {from=MaterialExpressionAdd_0, output=""}
    Normal {from=MaterialExpressionAdd_0, output=""}
    Yell {from=MaterialExpressionConstant_16, output=""}
   ]
   properties {inputs [
     {
      inputName=Normal
      input {
       expression=&self:MaterialExpressionAdd_0
       inputName=None
       mask=0
       maskA=0
       maskB=0
       maskG=0
       maskR=0
       outputIndex=0
      }
     }
     {
      inputName=Yell
      input {
       expression=&self:MaterialExpressionConstant_16
       inputName=None
       mask=0
       maskA=0
       maskB=0
       maskG=0
       maskR=0
       outputIndex=0
      }
     }
    ]}
  }
  MaterialExpressionTextureCoordinate_1 {class=MaterialExpressionTextureCoordinate, properties {coordinateIndex=1}}
  MaterialExpressionTextureCoordinate_3 {class=MaterialExpressionTextureCoordinate}
  MaterialExpressionTextureSample_1 {
   class=MaterialExpressionTextureSample
   inputs [
    UVs {from=MaterialExpressionMaterialFunctionCall_1, output=FlipbookUV1}
    Tex {from=""}
    Apply View MipBias {from=""}
   ]
   properties {
    samplerSource=SSM_Clamp_WorldGroupSettings
    texture=/BRCosmetics/Accessories/FORT_Backpacks/Backpack_M_MED_HonorBraceJoust/Textures/T_HonorBraceJoust_Backpack_Flipbook
   }
  }
  MaterialExpressionTextureSample_2 {
   class=MaterialExpressionTextureSample
   inputs [
    UVs {from=MaterialExpressionMaterialFunctionCall_7, output=FlipbookUV1}
    Tex {from=""}
    Apply View MipBias {from=""}
   ]
   properties {
    samplerSource=SSM_Clamp_WorldGroupSettings
    texture=&T_HonorBraceJoust_Backpack_Flipbook
   }
  }
  MaterialExpressionTextureSample_5 {
   class=MaterialExpressionTextureSample
   inputs [
    UVs {from=MaterialExpressionMaterialFunctionCall_0, output=FlipbookUV1}
    Tex {from=""}
    Apply View MipBias {from=""}
   ]
   properties {
    samplerSource=SSM_Clamp_WorldGroupSettings
    texture=&T_HonorBraceJoust_Backpack_Flipbook
   }
  }
  MaterialExpressionTime_0 {class=MaterialExpressionTime}
  MaterialExpressionTime_1 {class=MaterialExpressionTime}
  MaterialExpressionTime_2 {class=MaterialExpressionTime, properties {bOverride_Period=true, period=1}}
  MaterialExpressionTime_3 {class=MaterialExpressionTime}
  MaterialExpressionTime_4 {class=MaterialExpressionTime}
 }}
